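# Niagara river effect that spawns schools of mesh fish on the GPU, with user-driven spawn rate, particle cap, speed, spread, depth and color variants.
# Fish spawn in a cylinder with initial X velocity, then follow spline attraction, curl noise and distance-field avoidance, with forces flattened to XY.
# A lifetime curve scales fish in and out, while randomized values and a waveform feed dynamic material parameters for per-fish variation.
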NiagaraSystem /Game/Effects/Fort_Effects/Effects/Environment/River/NS_Env_River_Fish
dependencies {usedDataInterfaces [&NiagaraDataInterfaceCurve], usedDynamicInputs [], usedModules [&SystemState]}
emitters {Fish_High {
  simTarget=GPUComputeSim
  emitterUpdateScript [
   &EmitterState(
    Enable Distance Culling=true
    Life Cycle Mode=::NewEnumerator1
    Loop Duration=1
    Module.bApplySpawnCountScaleByDistance=false
    Module.bUseMaxDistance=true
    Module.bUseMinDistance=true
    Scalability Mode=::NewEnumerator1
    Scale Spawn Count=true
    Spawn Count Scale By Distance=DataInterface() {propertyValues {Curve {keys [{tangentMode=RCTM_User, time=0, value=1}, {tangentMode=RCTM_User, time=1, value=0}]}}}
    UseLoopDelay=false
   )
   &SpawnRate(
    Spawn Probability=0.25
    SpawnRate=&Multiply_Float(
     A=1
     B=&MakeCustomFloatFromBool(
      Bool=&SetBoolByIntComparison(A=$Engine.Emitter.NumParticles, B=$User.SpawnCap, Comparison Type=::NewEnumerator0)
      True Float=0
      False Float=$User.SpawnRate
     )
    )
    Use Spawn Probability=true
   ) {module=SpawnRate001}
  ]
  particleSpawnScript [
   &InitializeParticle(
    Color=&SelectLinearColorFromArray(Array Sampling Mode=::NewEnumerator0, Direct Array Index=?, Color Selection Array=$User.ColorVars, Random Seed=0)
    Lifetime Max=10
    Lifetime Min=8
    Lifetime Mode=::NewEnumerator1
    Mesh Scale Mode=::NewEnumerator3
    Mesh Uniform Scale=0.66
    Sprite Size=<28 20>
    Sprite Size Mode=::NewEnumerator4
    UsePositionOffset=false
    Write Material Random=true
   )
   &AddVelocity(Velocity=&RandomRangeVector(
     Minimum=&MakeVector(X=$User.MinVelocity, Y=0, Z=0)
     Maximum=&MakeVector(X=$User.MaxVelocity, Y=0, Z=0)
    ))
   &ApplyInitialForces()
   &InitialMeshOrientation(
    Mesh Axis To Orient=<0 1 0>
    Mesh Orientation Mode=::NewEnumerator2
    Rotation=&RandomRangeVector(Minimum=<0 0 0.25>, Maximum=<0 0 0.25>)
    Use Rotation Vector=true
   )
   &ShapeLocation(
    Cylinder Height=$User.SpawnDepth
    Cylinder Height Midpoint=1
    Cylinder Radius=$User.RadiusSpread
    MaskSpawnGroup=false
    Override Seed=false
    Shape Primitive=::NewEnumerator1
    Surface Only=false
    UseOwnerScale=true
   )
  ]
  particleUpdateScript [
   &ParticleState()
   /Niagara/Modules/Math/AvoidDistanceFieldSurfaces_GPU(
    Avoid Nearest Surface=true
    Avoid Oncoming Surfaces=true
    Falloff Distance=5000
    Max Trace Distance=2500
    Nearest Surface Avoidance Strength=50
    Obstactle Avoidance Strength=1
   )
   &ScaleMeshSize(Scale Factor=&VectorFromFloat(Value=&FloatFromCurve(
      FloatCurve=DataInterface() {propertyValues {Curve {keys [
          {interpMode=RCIM_Cubic, tangentMode=RCTM_User, time=0, value=0}
          <at 0.1 10>
          {interpMode=RCIM_Cubic, tangentMode=RCTM_User, time=0.9, value=10}
          {tangentMode=RCTM_User, time=1, value=0}
         ]}}}
      Scale Curve=0.85
     )))
   &CurlNoiseForce(Noise Strength=100, Pan Noise=false)
   &PointAttractionForce(
    Attraction Radius:NiagaraFloat=$User.RadiusSpread
    Attraction Strength=25
    Attractor Position:NiagaraPosition=&self:SplineAttractor(Spline:NiagaraDataInterfaceSpline=$User.Spline)
    Falloff Exponent=0.25
    Kill Within Radius=false
    Module.Use Falloff=true
    UseAttractorOffset=false
   )
   &Drag(Drag=&RandomRangeFloat(Minimum=0, Maximum=0.25), Use Linear Drag=true, Use Rotational Drag=true)
   &DynamicMaterialParameters(
    Index 0 Param 1=&RandomRangeFloat(Minimum=1, Maximum=1)
    Index 0 Param 2=&RandomRangeFloat(Minimum=0.4, Maximum=0.7)
    Index 0 Param 3=&RandomRangeFloat(Minimum=0, Maximum=1)
    Index 0 Param 4=&Waveform(
     Waveform Count=::NewEnumerator1
     Global Amplitude Scale=7
     "Waveform [1]"=::NewEnumerator0
     "[1] Frequency"=&RandomRangeFloat(Minimum=1, Maximum=2)
     "[1] Amplitude Min/Max"=<0.5 1>
     "Waveform [2] Blend"=::NewEnumerator2
     "[2] Input Position"=0
     "[2] Phase"=$Emitter.Age
     "[2] Phase Scale"=1
     "[2] Compound Frequency"=2
     "[2] Frequency"=&RandomRangeFloat(Minimum=0.7, Maximum=1)
     "[2] Amplitude Min/Max"=<0.1 0.2>
     "[2] Scale"=1
     "[2] Bias"=0
     "[2] Exponent"=1
     "[2] Pulse Duration"=0.1
     "[2] Clip Triangle Min/Max"=<0 0>
     "[2] Random Seed"=0
    )
    Index 1 Param 1=&RandomRangeFloat(Minimum=0, Maximum=0.25)
    Index 1 Param 2=&RandomRangeFloat(Minimum=5000, Maximum=15000)
    Index 1 Param 3=1
    Module.Param0WriteEnabled=true
    Module.Param0WriteEnabled1=true
    Module.Param1WriteEnabled=true
    Module.Param1WriteEnabled1=true
    Module.Param2WriteEnabled=true
    Module.Param2WriteEnabled1=true
    Module.Param3WriteEnabled=true
    Module.Param3WriteEnabled1=true
    Write Parameter Index 1=true
   )
   &SolveForcesAndVelocity(
    Clamp Velocity=false
    Force:Vector3f=&Multiply_Vector(A=$Transient.PhysicsForce, B=<1 1 0>)
    Limit Acceleration=false
   )
   &CalculateAccurateVelocity(Calculate Accurate Velocity On First Frame=true)
   &ScaleColor(
    Scale RGB=&Multiply_Vector(
     A=&Lerp_Vector(A=<1 1 1>, B=<1 1 1>, Alpha=$Output.RegionMask.RegionMask)
     B=&VectorFromFloat(Value=&ClampFloat(Float=$Output.RegionMask.RegionMask, Min=0.6, Max=25))
    )
    ScaleA=true
    ScaleRGB=true
   ) {module=ScaleColor001}
  ]
  rendererClasses [&NiagaraMeshRendererProperties]
  renderers [{rendererClass=&NiagaraMeshRendererProperties, rendererIndex=0}]
 }}
summary {
 systemName=NS_Env_River_Fish
 emitters [{emitterName=Fish_High, simTarget=GPUComputeSim, rendererClasses [&NiagaraMeshRendererProperties]}]
 userVariables [
  User.ColorVars {
   type=NiagaraDataInterfaceArrayColor
   description=""
   defaultValue {
    struct=NiagaraExt_VariableValue_DataInterface
    value {
     dataInterface=&self:NiagaraDataInterfaceArrayColor_0
     dataInterfaceClass=&NiagaraDataInterfaceArrayColor
    }
   }
  }
  User.MaxVel {type=NiagaraFloat, defaultValue=250, description=""}
  User.MaxVelocity {type=NiagaraFloat, defaultValue=500, description=""}
  User.MinVelocity {type=NiagaraFloat, defaultValue=350, description=""}
  User.RadiusSpread {type=NiagaraFloat, defaultValue=500, description=""}
  User.SpawnCap {type=NiagaraInt32, defaultValue=2500}
  User.SpawnDepth {type=NiagaraFloat, defaultValue=125, description=""}
  User.SpawnRate {type=NiagaraFloat, defaultValue=5, description=""}
  User.Spline {
   type=NiagaraDataInterfaceSpline
   description=""
   defaultValue {
    struct=NiagaraExt_VariableValue_DataInterface
    value {
     dataInterface=&self:NiagaraDataInterfaceSpline_1
     dataInterfaceClass=/Script/Niagara.NiagaraDataInterfaceSpline
    }
   }
  }
 ]
}
